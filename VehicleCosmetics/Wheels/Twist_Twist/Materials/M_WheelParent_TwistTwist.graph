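# Parent material for the Twist Twist vehicle wheel cosmetic, overriding Emissive on MF_WheelParent_Simple before damage and de-rez transition functions.
# Rim glow uses a time-panned gradient to flip a honeycomb mask between normal and inverted, masked by an emissive texture and tinted by CustomColor.
# Tread glow combines two Fresnel functions with tiled tire-mask channels, tinted by CustomColorTier1.
# Vertex color red blends between rim and tread glow.
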Material /VehicleCosmetics/Wheels/Twist_Twist/Materials/M_WheelParent_TwistTwist
graph {
 properties {
  bAutomaticallySetUsageInEditor=false
  bUseMaterialAttributes=true
  bUsedWithMeshDeformer=true
  bUsedWithSkeletalMesh=true
  pixelDepthOffsetMode=PDOM_Legacy
  usedShadingModels=Default Lit
 }
 nodes [
  n1 = GetMaterialAttributes(None=MaterialFunctionCall() {materialFunction=&MF_WheelParent_Simple}.MaterialAttributes) {attributeGetTypes [#1]}
  n2 = NamedRerouteUsage()
  MaterialFunctionCall(In=MaterialFunctionCall(Material=SetMaterialAttributes(MaterialAttributes=@n1.MaterialAttributes, Emissive Color=@n2) {attributeSetTypes [#1]}) {materialFunction=&MF_BaseVehicle_Damage}.Result) {materialFunction=&MF_VehicleDeRez_Transition}
  n3 = TextureSample(UVs=TextureCoordinate() {uTiling=20, vTiling=20}) {samplerType=SAMPLERTYPE_LinearColor, texture=/VehicleCosmetics/Wheels/Twist_Twist/Textures/Honeycomb_mask}
  n4 = TextureCoordinate()
  n5 = VectorParameter() {parameterName=PulsePannerTiming, defaultValue=<rgba 1 3 -1 1>}
  n6 = AppendVector(A=Multiply(A=ComponentMask(None=@n4) {r=true}, B=@n5.R), B=Multiply(A=ComponentMask(None=@n4) {g=true}, B=@n5.G))
  n7 = Multiply(A=Time(), B=@n5.B)
  n8 = LinearInterpolate(
   A=@n3.R
   B=OneMinus(None=@n3.R)
   Alpha=TextureSample(UVs=Panner(Coordinate=@n6, Time=@n7) {speedY=1}) {samplerType=SAMPLERTYPE_LinearColor, texture=/VehicleCosmetics/Wheels/Twist_Twist/Textures/Glow_Gradient_Alpha}.R
  )
  n9 = Constant() {r=0.25}
  n10 = TextureSampleParameter2D() {
   parameterName=Emissive
   samplerType=SAMPLERTYPE_LinearColor
   texture=/VehicleCosmetics/Wheels/Twist_Twist/Textures/T_Twist_Twist_Rim_Emissive
  }
  n11 = Multiply(A=Multiply(A=Add(A=@n8, B=@n9), B=@n10.R), B=VectorParameter() {parameterName=CustomColor, defaultValue=<rgba 1 0.0685 0.0685 0>}.RGB)
  n12 = ScalarParameter() {defaultValue=4, parameterName=PulseEmissive}
  n13 = MaterialFunctionCall!(Normal, Exponent) {materialFunction=/VehicleCosmetics/Wheels/Cola_Holographic/Materials/MF_Fresnel_Classic}
  n14 = VectorParameter() {group=5_Special Effects, parameterName=CustomColorTier1, defaultValue=<rgba 0.1779 0.1413 0.8714 1>}
  n15 = TextureCoordinate()
  n16 = ComponentMask(None=@n15) {r=true}
  n17 = VectorParameter() {group=5_Special Effects, parameterName=TreadTiling, channelNames {a="", b="", g=V-Tiling, r=U-Tiling}, defaultValue=<rgba 5 1 0 1>}
  n18 = ComponentMask(None=@n15) {g=true}
  n19 = TextureSample(UVs=AppendVector(A=Multiply(A=@n16, B=@n17.U-Tiling), B=Multiply(A=@n17.V-Tiling, B=@n18))) {samplerType=SAMPLERTYPE_LinearColor, texture=/VehicleCosmetics/Wheels/Twist_Twist/Textures/T_Twist_Twist_Tire_M}
  n20 = Multiply(
   A=Add(A=Multiply(A=@n13.Result, B=@n14.RGB), B=Multiply(A=@n19.B, B=@n14.RGB))
   B=ScalarParameter() {defaultValue=5, group=5_Special Effects, parameterName=TreadGlowPower}
  )
  n21 = Multiply(
   A=MaterialFunctionCall!(
    Input Normal=TextureSample() {samplerType=SAMPLERTYPE_LinearColor, texture=/VehicleCosmetics/Wheels/Twist_Twist/Textures/T_Twist_Twist_Tire_N}.RGB
    Input Sharpness=ScalarParameter() {defaultValue=1, group=5_Special Effects, parameterName=Sharpness}
    Input Two Sided
   ) {materialFunction=/VehicleCosmetics/Wheels/CarCar_Holographic/Materials/MF_Fresnel_Advanced}.Middle
   B=Multiply(A=@n19.R, B=@n14.RGB)
  )
  NamedRerouteDeclaration(None=LinearInterpolate(A=Multiply(A=@n11, B=@n12), B=LinearInterpolate(A=@n20, B=@n21, Alpha=@n19.R), Alpha=VertexColor().R)) {name=Emissive, nodeColor=<rgba 1 0.75 0 1>}
 ]
}
