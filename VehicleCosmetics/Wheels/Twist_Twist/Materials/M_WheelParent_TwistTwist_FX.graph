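# Translucent lit parent material for the Twist_Twist skeletal vehicle wheel, with an animated, color-tintable glow effect.
# A panning gradient blends a honeycomb mask with its inverse, masked by the rim emissive texture to form a pulse.
# Vertex color red blends that pulse into a flat boosted glow.
# The same channel blends a solid value into two panning caustic layers shaped by a shockwave curve texture, borrowing Secretooze's caustics.
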Material /VehicleCosmetics/Wheels/Twist_Twist/Materials/M_WheelParent_TwistTwist_FX
graph {
 properties {
  bAutomaticallySetUsageInEditor=false
  bUsedWithSkeletalMesh=true
  blendMode=BLEND_Translucent
  previewMesh=/VehicleCosmetics/Wheels/Twist_Twist/Meshes/SK_Wheel_Twist_Twist
  usedShadingModels=Default Lit
 }
 nodes [
  n1 = TextureSample(UVs=TextureCoordinate() {uTiling=20, vTiling=20}) {samplerType=SAMPLERTYPE_LinearColor, texture=/VehicleCosmetics/Wheels/Twist_Twist/Textures/Honeycomb_mask}
  n2 = OneMinus(None=@n1.R)
  n3 = TextureCoordinate()
  n4 = VectorParameter() {parameterName=PulsePannerTiming, defaultValue=<rgba 1 3 -1 1>}
  n5 = Multiply(A=ComponentMask(None=@n3) {r=true}, B=@n4.R)
  n6 = Multiply(A=ComponentMask(None=@n3) {g=true}, B=@n4.G)
  n7 = Time()
  n8 = TextureSample(UVs=Panner(Coordinate=AppendVector(A=@n5, B=@n6), Time=Multiply(A=@n7, B=@n4.B)) {speedY=1}) {samplerType=SAMPLERTYPE_LinearColor, texture=/VehicleCosmetics/Wheels/Twist_Twist/Textures/Glow_Gradient_Alpha}
  n9 = TextureSampleParameter2D() {
   parameterName=Emissive
   samplerType=SAMPLERTYPE_LinearColor
   texture=/VehicleCosmetics/Wheels/Twist_Twist/Textures/T_Twist_Twist_Rim_Emissive
  }
  n10 = Multiply(A=Add(A=LinearInterpolate(A=@n1.R, B=@n2, Alpha=@n8.R), B=Constant() {r=0.25}), B=@n9.R)
  n11 = VectorParameter() {parameterName=CustomColor, defaultValue=<rgba 1 0.0685 0.0685 0>}
  n12 = VertexColor()
  LinearInterpolate(
   A=Multiply(A=Multiply(A=@n10, B=@n11.RGBA), B=ScalarParameter() {defaultValue=4, parameterName=PulseEmissive})
   B=Multiply(A=ScalarParameter() {defaultValue=25, parameterName=EmissiveGlowMultiply}, B=@n11.RGBA)
   Alpha=@n12.R
  )
  n13 = TextureCoordinate()
  n14 = AppendVector(
   A=Multiply(A=ComponentMask(None=@n13) {r=true}, B=ScalarParameter() {defaultValue=1, parameterName=TileX})
   B=Multiply(A=ComponentMask(None=@n13) {g=true}, B=ScalarParameter() {defaultValue=1, parameterName=TileY})
  )
  n15 = Multiply(A=Time(), B=ScalarParameter() {defaultValue=1, parameterName=PanSpeedAdjust})
  n16 = TextureSampleParameter2D(UVs=Panner(Coordinate=@n14, Time=@n15) {speedX=0.0754, speedY=0.0675}) {parameterName=Caustic_MAP1, texture=/VehicleCosmetics/Wheels/Secretooze/Textures/T_Caustics_D}
  n17 = Add(
   A=TextureSampleParameter2D(UVs=Panner(Coordinate=@n14, Time=@n15) {speedX=0.125, speedY=1}) {parameterName=Caustic_MAP2, texture=&T_Caustics_D}.R
   B=Multiply(A=@n16.G, B=@n16.G)
  )
  n18 = TextureSample() {texture=/VehicleCosmetics/Wheels/_SharedTextures/FX/Curve_Shockwaves_Pack}
  n19 = Constant() {r=0.025}
  LinearInterpolate(A=Constant() {r=1}, B=Add(A=Multiply(A=@n17, B=@n18.R), B=Multiply(A=@n18.R, B=@n19)), Alpha=@n12.R)
  Multiply(A=Multiply(A=@n16.R, B=@n18.R), B=Constant() {r=32})
  ScalarParameter() {defaultValue=1, parameterName=Roughness}
  ScalarParameter() {parameterName=Metallic}
 ]
}
